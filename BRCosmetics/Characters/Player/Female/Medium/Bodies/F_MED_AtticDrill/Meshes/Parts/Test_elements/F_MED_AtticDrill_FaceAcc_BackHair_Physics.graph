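# Physics asset for the back-hair accessory of the AtticDrill female medium character, stored under Test_elements.
# Three capsule chains (center, left, right) each run from a base bone tied to spine_05 through eight dyn_backHair bones.
# Capsules on pelvis, spine, clavicles, arms and thighs add body volumes, linked by free-motion constraints.
# Chain roots also move freely, while joints from bone 03 down carry asymmetric swing limits and zero twist limit.
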PhysicsAsset /BRCosmetics/Characters/Player/Female/Medium/Bodies/F_MED_AtticDrill/Meshes/Parts/Test_elements/F_MED_AtticDrill_FaceAcc_BackHair_Physics
bodies {
 base_backHair_c [{shapeType=Capsule, length=1.7018, radius=1.3896, shapeName=base_backHair_c_capsule}]
 base_backHair_l [{shapeType=Capsule, length=2.3651, radius=2.3849, shapeName=base_backHair_l_capsule}]
 base_backHair_r [{shapeType=Capsule, length=2.6005, radius=1.0042, shapeName=base_backHair_r_capsule}]
 clavicle_l [{shapeType=Capsule, length=4, radius=10, shapeName=clavicle_l_capsule, center=<8.7436 0 -5.7334>, rotation=<rot 90 0 0>}]
 clavicle_r [{shapeType=Capsule, length=4, radius=10, shapeName=clavicle_r_capsule, center=<-8.7436 0 5.7334>, rotation=<rot 90 180 0>}]
 dyn_backHair_c_01 [{shapeType=Capsule, length=1, radius=1, shapeName=dyn_backHair_c_01_capsule}]
 dyn_backHair_c_02 [{shapeType=Capsule, length=1, radius=1, shapeName=dyn_backHair_c_02_capsule}]
 dyn_backHair_c_03 [{shapeType=Capsule, length=5, radius=5, shapeName=dyn_backHair_c_03_capsule, center=<2 0 1.7287>, rotation=<rot 90 0 0>}]
 dyn_backHair_c_04 [{shapeType=Capsule, length=5, radius=5, shapeName=dyn_backHair_c_04_capsule, center=<2 0 1.9316>, rotation=<rot 90 0 0>}]
 dyn_backHair_c_05 [{shapeType=Capsule, length=5, radius=5, shapeName=dyn_backHair_c_05_capsule, center=<2 0 0>, rotation=<rot 90 0 0>}]
 dyn_backHair_c_06 [{shapeType=Capsule, length=5, radius=5, shapeName=dyn_backHair_c_06_capsule, center=<2 0 0>, rotation=<rot 90 0 0>}]
 dyn_backHair_c_07 [{shapeType=Capsule, length=5, radius=5, shapeName=dyn_backHair_c_07_capsule, center=<2 0 0>, rotation=<rot 90 0 0>}]
 dyn_backHair_c_08 [{shapeType=Capsule, length=5, radius=5, shapeName=dyn_backHair_c_08_capsule, center=<2 0 0>, rotation=<rot 90 0 0>}]
 dyn_backHair_l_01 [{shapeType=Capsule, length=1, radius=1, shapeName=dyn_backHair_l_01_capsule}]
 dyn_backHair_l_02 [{shapeType=Capsule, length=1, radius=1, shapeName=dyn_backHair_l_02_capsule}]
 dyn_backHair_l_03 [{shapeType=Capsule, length=5, radius=5, shapeName=dyn_backHair_l_03_capsule, center=<2 0 1.7287>, rotation=<rot 90 0 0>}]
 dyn_backHair_l_04 [{
   shapeType=Capsule
   length=9.6979
   radius=6.2448
   shapeName=dyn_backHair_l_04_capsule
   center=<1.1766 0.5512 1.6582>
   rotation=<rot 59.1388 146.2036 55.6209>
  }]
 dyn_backHair_l_05 [{shapeType=Capsule, length=13.0579, radius=5, shapeName=dyn_backHair_l_05_capsule, center=<2 0 0>, rotation=<rot 90 89.5681 0>}]
 dyn_backHair_l_06 [{shapeType=Capsule, length=15.3221, radius=5, shapeName=dyn_backHair_l_06_capsule, center=<2 0 0>, rotation=<rot 90 87.8884 0>}]
 dyn_backHair_l_07 [{shapeType=Capsule, length=5, radius=5, shapeName=dyn_backHair_l_07_capsule, center=<2 0 0>, rotation=<rot 90 0 0>}]
 dyn_backHair_l_08 [{shapeType=Capsule, length=5, radius=5, shapeName=dyn_backHair_l_08_capsule, center=<2 0 0>, rotation=<rot 90 0 0>}]
 dyn_backHair_r_01 [{shapeType=Capsule, length=1, radius=1, shapeName=dyn_backHair_r_01_capsule}]
 dyn_backHair_r_02 [{shapeType=Capsule, length=1, radius=1, shapeName=dyn_backHair_r_02_capsule}]
 dyn_backHair_r_03 [{shapeType=Capsule, length=5, radius=5, shapeName=dyn_backHair_r_03_capsule, center=<2 0 1.7287>, rotation=<rot 90 0 0>}]
 dyn_backHair_r_04 [{shapeType=Capsule, length=9.6979, radius=6.2448, shapeName=dyn_backHair_r_04_capsule, center=<2 0 1.9316>, rotation=<rot 90 88.6523 0>}]
 dyn_backHair_r_05 [{shapeType=Capsule, length=13.0579, radius=5, shapeName=dyn_backHair_r_05_capsule, center=<2 0 0>, rotation=<rot 90 89.5681 0>}]
 dyn_backHair_r_06 [{shapeType=Capsule, length=15.3221, radius=5, shapeName=dyn_backHair_r_06_capsule, center=<2 0 0>, rotation=<rot 90 87.8884 0>}]
 dyn_backHair_r_07 [{shapeType=Capsule, length=5, radius=5, shapeName=dyn_backHair_r_07_capsule, center=<2 0 0>, rotation=<rot 90 0 0>}]
 dyn_backHair_r_08 [{shapeType=Capsule, length=5, radius=5, shapeName=dyn_backHair_r_08_capsule, center=<2 0 0>, rotation=<rot 90 0 0>}]
 lowerarm_l [{shapeType=Capsule, length=15, radius=5.6764, shapeName=lowerarm_l_capsule, center=<15 0 -0.6443>, rotation=<rot 90 -14 0>}]
 lowerarm_r [{shapeType=Capsule, length=15, radius=5.6764, shapeName=lowerarm_r_capsule, center=<-15 0 0.6443>, rotation=<rot 90 166 0>}]
 pelvis [{shapeType=Capsule, length=27, radius=15, shapeName=pelvis_capsule, center=<-0.1961 -2.3902 0>, rotation=<rot 0 -4.6913 0>}]
 spine_01 [{shapeType=Capsule, length=22.5, radius=15, shapeName=spine_01_capsule, center=<0 -1.4618 0>}]
 spine_02 [{shapeType=Capsule, length=16.5, radius=13.5, shapeName=spine_02_capsule, center=<0 -0.8444 0>}]
 spine_03 [{shapeType=Capsule, length=21.847, radius=10.2111, shapeName=spine_03_capsule, center=<0.4408 -3.431 0>, rotation=<rot 0 -82.6793 0>}]
 spine_04 [{shapeType=Capsule, length=10, radius=13, shapeName=spine_04_capsule}]
 spine_05 [{shapeType=Capsule, length=10, radius=12, shapeName=spine_05_capsule, center=<0 -1.2571 0>}]
 thigh_l [{shapeType=Capsule, length=17.505, radius=8.005, shapeName=thigh_l_capsule, center=<-7.9854 0 -0.4839>, rotation=<rot 0 -90 88.7055>}]
 thigh_r [{shapeType=Capsule, length=17.505, radius=8.005, shapeName=thigh_r_capsule, center=<7.9854 0 0.4839>, rotation=<rot 0 -90 -91.2945>}]
 thigh_twist_01_l [{shapeType=Capsule, length=17.505, radius=8.005, shapeName=thigh_l_capsule, center=<-7.9854 0 -0.4839>, rotation=<rot 0 -90 88.7055>}]
 thigh_twist_01_r [{shapeType=Capsule, length=17.505, radius=8.005, shapeName=thigh_r_capsule, center=<7.9854 0 0.4839>, rotation=<rot 0 -90 -91.2945>}]
 upperarm_l [{shapeType=Capsule, length=20, radius=7, shapeName=upperarm_r_capsule, center=<15 0 0>, rotation=<rot 90 0 0>}]
 upperarm_r [{shapeType=Capsule, length=20, radius=7, shapeName=upperarm_l_capsule, center=<-15 3.1379 0>, rotation=<rot 90 -171.1979 0>}]
}
constraints [
 {
  bone1Name=spine_01
  bone2Name=pelvis
  swing1LimitDegrees=45
  swing1Motion=Free
  swing2LimitDegrees=45
  swing2Motion=Free
  twistLimitDegrees=45
  twistMotion=Free
 }
 {
  bone1Name=spine_02
  bone2Name=spine_01
  swing1LimitDegrees=45
  swing1Motion=Free
  swing2LimitDegrees=45
  swing2Motion=Free
  twistLimitDegrees=45
  twistMotion=Free
 }
 {
  bone1Name=spine_03
  bone2Name=spine_02
  swing1LimitDegrees=45
  swing1Motion=Free
  swing2LimitDegrees=45
  swing2Motion=Free
  twistLimitDegrees=45
  twistMotion=Free
 }
 {
  bone1Name=spine_04
  bone2Name=spine_03
  swing1LimitDegrees=45
  swing1Motion=Free
  swing2LimitDegrees=45
  swing2Motion=Free
  twistLimitDegrees=45
  twistMotion=Free
 }
 {
  bone1Name=spine_05
  bone2Name=spine_04
  swing1LimitDegrees=45
  swing1Motion=Free
  swing2LimitDegrees=45
  swing2Motion=Free
  twistLimitDegrees=45
  twistMotion=Free
 }
 {
  bone1Name=clavicle_l
  bone2Name=spine_05
  swing1LimitDegrees=45
  swing1Motion=Free
  swing2LimitDegrees=45
  swing2Motion=Free
  twistLimitDegrees=45
  twistMotion=Free
 }
 {
  bone1Name=upperarm_l
  bone2Name=clavicle_l
  swing1LimitDegrees=45
  swing1Motion=Free
  swing2LimitDegrees=45
  swing2Motion=Free
  twistLimitDegrees=45
  twistMotion=Free
 }
 {
  bone1Name=lowerarm_l
  bone2Name=upperarm_l
  swing1LimitDegrees=45
  swing1Motion=Free
  swing2LimitDegrees=45
  swing2Motion=Free
  twistLimitDegrees=45
  twistMotion=Free
 }
 {
  bone1Name=clavicle_r
  bone2Name=spine_05
  swing1LimitDegrees=45
  swing1Motion=Free
  swing2LimitDegrees=45
  swing2Motion=Free
  twistLimitDegrees=45
  twistMotion=Free
 }
 {
  bone1Name=upperarm_r
  bone2Name=clavicle_r
  swing1LimitDegrees=45
  swing1Motion=Free
  swing2LimitDegrees=45
  swing2Motion=Free
  twistLimitDegrees=45
  twistMotion=Free
 }
 {
  bone1Name=lowerarm_r
  bone2Name=upperarm_r
  swing1LimitDegrees=45
  swing1Motion=Free
  swing2LimitDegrees=45
  swing2Motion=Free
  twistLimitDegrees=45
  twistMotion=Free
 }
 {
  bone1Name=thigh_l
  bone2Name=pelvis
  swing1LimitDegrees=45
  swing1Motion=Free
  swing2LimitDegrees=45
  swing2Motion=Free
  twistLimitDegrees=45
  twistMotion=Free
 }
 {
  bone1Name=thigh_twist_01_l
  bone2Name=thigh_l
  swing1LimitDegrees=45
  swing1Motion=Free
  swing2LimitDegrees=45
  swing2Motion=Free
  twistLimitDegrees=45
  twistMotion=Free
 }
 {
  bone1Name=thigh_r
  bone2Name=pelvis
  swing1LimitDegrees=45
  swing1Motion=Free
  swing2LimitDegrees=45
  swing2Motion=Free
  twistLimitDegrees=45
  twistMotion=Free
 }
 {
  bone1Name=thigh_twist_01_r
  bone2Name=thigh_r
  swing1LimitDegrees=45
  swing1Motion=Free
  swing2LimitDegrees=45
  swing2Motion=Free
  twistLimitDegrees=45
  twistMotion=Free
 }
 {bone1Name=dyn_backHair_c_03, bone2Name=dyn_backHair_c_02, swing1LimitDegrees=10, swing2LimitDegrees=45, twistLimitDegrees=0}
 {bone1Name=dyn_backHair_c_04, bone2Name=dyn_backHair_c_03, swing1LimitDegrees=10, swing2LimitDegrees=45, twistLimitDegrees=0}
 {bone1Name=dyn_backHair_c_05, bone2Name=dyn_backHair_c_04, swing1LimitDegrees=10, swing2LimitDegrees=45, twistLimitDegrees=0}
 {bone1Name=dyn_backHair_c_06, bone2Name=dyn_backHair_c_05, swing1LimitDegrees=10, swing2LimitDegrees=45, twistLimitDegrees=0}
 {bone1Name=dyn_backHair_c_07, bone2Name=dyn_backHair_c_06, swing1LimitDegrees=10, swing2LimitDegrees=45, twistLimitDegrees=0}
 {bone1Name=dyn_backHair_c_08, bone2Name=dyn_backHair_c_07, swing1LimitDegrees=10, swing2LimitDegrees=45, twistLimitDegrees=0}
 {bone1Name=dyn_backHair_l_03, bone2Name=dyn_backHair_l_02, swing1LimitDegrees=10, swing2LimitDegrees=45, twistLimitDegrees=0}
 {bone1Name=dyn_backHair_l_04, bone2Name=dyn_backHair_l_03, swing1LimitDegrees=10, swing2LimitDegrees=45, twistLimitDegrees=0}
 {bone1Name=dyn_backHair_l_05, bone2Name=dyn_backHair_l_04, swing1LimitDegrees=10, swing2LimitDegrees=45, twistLimitDegrees=0}
 {bone1Name=dyn_backHair_l_06, bone2Name=dyn_backHair_l_05, swing1LimitDegrees=10, swing2LimitDegrees=45, twistLimitDegrees=0}
 {bone1Name=dyn_backHair_l_07, bone2Name=dyn_backHair_l_06, swing1LimitDegrees=10, swing2LimitDegrees=45, twistLimitDegrees=0}
 {bone1Name=dyn_backHair_l_08, bone2Name=dyn_backHair_l_07, swing1LimitDegrees=10, swing2LimitDegrees=45, twistLimitDegrees=0}
 {bone1Name=dyn_backHair_r_03, bone2Name=dyn_backHair_r_02, swing1LimitDegrees=10, swing2LimitDegrees=45, twistLimitDegrees=0}
 {bone1Name=dyn_backHair_r_04, bone2Name=dyn_backHair_r_03, swing1LimitDegrees=10, swing2LimitDegrees=45, twistLimitDegrees=0}
 {bone1Name=dyn_backHair_r_05, bone2Name=dyn_backHair_r_04, swing1LimitDegrees=10, swing2LimitDegrees=45, twistLimitDegrees=0}
 {bone1Name=dyn_backHair_r_06, bone2Name=dyn_backHair_r_05, swing1LimitDegrees=10, swing2LimitDegrees=45, twistLimitDegrees=0}
 {bone1Name=dyn_backHair_r_07, bone2Name=dyn_backHair_r_06, swing1LimitDegrees=10, swing2LimitDegrees=45, twistLimitDegrees=0}
 {bone1Name=dyn_backHair_r_08, bone2Name=dyn_backHair_r_07, swing1LimitDegrees=10, swing2LimitDegrees=45, twistLimitDegrees=0}
 {
  bone1Name=base_backHair_c
  bone2Name=spine_05
  swing1LimitDegrees=45
  swing1Motion=Free
  swing2LimitDegrees=45
  swing2Motion=Free
  twistLimitDegrees=45
  twistMotion=Free
 }
 {
  bone1Name=base_backHair_l
  bone2Name=spine_05
  swing1LimitDegrees=45
  swing1Motion=Free
  swing2LimitDegrees=45
  swing2Motion=Free
  twistLimitDegrees=45
  twistMotion=Free
 }
 {
  bone1Name=base_backHair_r
  bone2Name=spine_05
  swing1LimitDegrees=45
  swing1Motion=Free
  swing2LimitDegrees=45
  swing2Motion=Free
  twistLimitDegrees=45
  twistMotion=Free
 }
 {
  bone1Name=dyn_backHair_c_02
  bone2Name=dyn_backHair_c_01
  swing1LimitDegrees=45
  swing1Motion=Free
  swing2LimitDegrees=45
  swing2Motion=Free
  twistLimitDegrees=45
  twistMotion=Free
 }
 {
  bone1Name=dyn_backHair_c_01
  bone2Name=base_backHair_c
  swing1LimitDegrees=45
  swing1Motion=Free
  swing2LimitDegrees=45
  swing2Motion=Free
  twistLimitDegrees=45
  twistMotion=Free
 }
 {
  bone1Name=dyn_backHair_l_01
  bone2Name=base_backHair_l
  swing1LimitDegrees=45
  swing1Motion=Free
  swing2LimitDegrees=45
  swing2Motion=Free
  twistLimitDegrees=45
  twistMotion=Free
 }
 {
  bone1Name=dyn_backHair_l_02
  bone2Name=dyn_backHair_l_01
  swing1LimitDegrees=45
  swing1Motion=Free
  swing2LimitDegrees=45
  swing2Motion=Free
  twistLimitDegrees=45
  twistMotion=Free
 }
 {
  bone1Name=dyn_backHair_r_01
  bone2Name=base_backHair_r
  swing1LimitDegrees=45
  swing1Motion=Free
  swing2LimitDegrees=45
  swing2Motion=Free
  twistLimitDegrees=45
  twistMotion=Free
 }
 {
  bone1Name=dyn_backHair_r_02
  bone2Name=dyn_backHair_r_01
  swing1LimitDegrees=45
  swing1Motion=Free
  swing2LimitDegrees=45
  swing2Motion=Free
  twistLimitDegrees=45
  twistMotion=Free
 }
]
