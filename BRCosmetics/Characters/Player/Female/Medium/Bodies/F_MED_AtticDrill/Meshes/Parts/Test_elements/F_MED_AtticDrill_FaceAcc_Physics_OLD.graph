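# Physics asset for the face accessory of the F_MED_AtticDrill female character, kept in a test folder with an OLD suffix.
# It gives the pelvis, spine, upper neck, clavicles, upper arms and forearms one capsule each, with mirrored offsets on left and right.
# It has no constraints and no head, hand or leg bodies, so it forms an upper-body collision proxy rather than a jointed ragdoll.
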PhysicsAsset /BRCosmetics/Characters/Player/Female/Medium/Bodies/F_MED_AtticDrill/Meshes/Parts/Test_elements/F_MED_AtticDrill_FaceAcc_Physics_OLD
bodies {
 clavicle_l [{shapeType=Capsule, length=4, radius=10, shapeName=clavicle_l_capsule, center=<8.7436 0 -5.7334>, rotation=<rot 90 0 0>}]
 clavicle_r [{shapeType=Capsule, length=4, radius=10, shapeName=clavicle_l_capsule, center=<-8.7436 0 5.7334>, rotation=<rot 90 0 0>}]
 lowerarm_l [{shapeType=Capsule, length=15, radius=11, shapeName=lowerarm_l_capsule, center=<15 0 -3>, rotation=<rot 90 -14 0>}]
 lowerarm_r [{shapeType=Capsule, length=15, radius=11, shapeName=lowerarm_l_capsule, center=<-15 0 3>, rotation=<rot 90 -14 0>}]
 neck_02 [{shapeType=Capsule, length=10, radius=8, shapeName=neck_02_capsule, rotation=<rot 90 0 0>}]
 pelvis [{shapeType=Capsule, length=27, radius=15, shapeName=pelvis_capsule}]
 spine_01 [{shapeType=Capsule, length=22.5, radius=15, shapeName=spine_01_capsule}]
 spine_02 [{shapeType=Capsule, length=16.5, radius=13.5, shapeName=spine_02_capsule}]
 spine_04 [{shapeType=Capsule, length=10, radius=13, shapeName=spine_04_capsule}]
 spine_05 [{shapeType=Capsule, length=10, radius=12, shapeName=spine_05_capsule, center=<0 -1.2571 0>}]
 upperarm_l [{shapeType=Capsule, length=20, radius=7, shapeName=upperarm_r_capsule, center=<15 0 0>, rotation=<rot 90 0 0>}]
 upperarm_r [{shapeType=Capsule, length=20, radius=7, shapeName=upperarm_r_capsule, center=<-15 0 0>, rotation=<rot 90 0 0>}]
}
constraints []
